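# Unlit, two-sided translucent material for a skull detail trail effect on Niagara mesh particles.
# World position offset combines panning smoke noise along vertex normals, growing along U, with a local-space push and slight pull toward object center.
# Opacity blends noise-distorted, scrolling caustics and smoke, shaped by U gradients, view-angle fade, particle alpha and TrailOpacity.
# Emissive multiplies particle color by a Color parameter with variant-driven desaturation.
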Material /Game/Effects/Fort_Effects/GameplayPulginsAssets/Skull/SkullDetail/MI_Detail_Skull_Trail
graph {
 properties {
  bUsedWithNiagaraMeshParticles=true
  blendMode=BLEND_Translucent
  pixelDepthOffsetMode=PDOM_Legacy
  shadingModel=MSM_Unlit
  twoSided=true
  usedShadingModels=Unlit
 }
 nodes [
  n1 = ObjectPositionWS()
  n2 = WorldPosition()
  n3 = ComponentMask(None=TextureCoordinate()) {r=true}
  n4 = Panner(Coordinate=TextureCoordinate() {uTiling=0.4}) {speedX=-0.5}
  n5 = Multiply(
   A=Multiply(A=TextureSample(UVs=@n4) {texture=&T_Inky_Smoke_Tile}.R, B=VertexNormalWS())
   B=ScalarParameter() {defaultValue=15, parameterName=Scale WPO}
  ) {constB=15}
  n6 = Constant3Vector() {constant=<rgba 0 3 0 0>}
  Add(
   A=Add(A=Multiply(A=@n3, B=@n5), B=Transform(None=@n6) {transformSourceType=TRANSFORMSOURCE_Local})
   B=Multiply(A=Subtract(A=@n1, B=@n2.XYZ)) {constB=0.2}
  ) {constB=0}
  n7 = VectorParameter() {parameterName=Color, defaultValue=<rgba 0 1.5 0.1969 1>}
  n8 = LinearInterpolate(A=ScalarParameter() {parameterName=Des}, Alpha=ScalarParameter() {parameterName=ColorVariant}) {constB=0}
  MaterialFunctionCall(Emissive=Multiply(A=ParticleColor().RGB, B=Desaturation(None=@n7.RGB, Fraction=@n8))) {materialFunction=&MF_FX_EmissiveAdjustments}
  n9 = TextureSample(UVs=Panner(Coordinate=TextureCoordinate() {uTiling=0.4})) {samplerType=SAMPLERTYPE_Grayscale, texture=&LowResBlurredNoise}
  n10 = Panner(Coordinate=Add(A=TextureCoordinate() {uTiling=0.2}, B=Multiply(A=@n9.R) {constB=0.2})) {speedX=-0.5}
  n11 = TextureSample(UVs=@n10) {samplerType=SAMPLERTYPE_LinearGrayscale, texture=/Game/Global/Textures/Default/Caustics/T_Nosie_Caustics_Grayscale}
  n12 = TextureSample(UVs=@n10) {texture=&T_Inky_Smoke_Tile}
  n13 = LinearInterpolate(Alpha=ComponentMask(None=TextureCoordinate()) {r=true}) {constA=-0.5}
  n14 = Add(A=Power(Base=LinearInterpolate(A=@n11.R, B=@n12.R), Exp=LinearInterpolate(Alpha=@n13) {constA=0.5, constB=2})) {constB=0}
  n15 = TextureCoordinate()
  n16 = Power(Base=OneMinus(None=ComponentMask(None=@n15) {r=true}))
  n17 = Multiply(
   A=Multiply(A=Multiply(A=@n14, B=@n16), B=Power(Base=@n3))
   B=MaterialFunctionCall("FadeSensitivity (optional)"=ScalarParameter() {defaultValue=2, parameterName="FadeSensitivity (optional) (S)"}) {materialFunction=&FadeBasedOnViewAngle}.Result
  )
  Multiply(A=Multiply(A=Multiply(A=@n17) {constB=1.3}, B=ParticleColor().A), B=ScalarParameter() {defaultValue=1, parameterName=TrailOpacity})
 ]
}
