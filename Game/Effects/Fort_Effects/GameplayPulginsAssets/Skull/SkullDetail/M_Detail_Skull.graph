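# Translucent Default Lit material for a skull detail mesh rendered by Niagara mesh particles.
# A skull texture passes through a parameterized desaturation and is tinted by particle color, with one branch scaled by an Emissive parameter into the shared FX emissive adjustments function.
# A separate branch multiplies particle alpha by a view-angle fade with tunable sensitivity.
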Material /Game/Effects/Fort_Effects/GameplayPulginsAssets/Skull/SkullDetail/M_Detail_Skull
graph {
 properties {
  bUsedWithNiagaraMeshParticles=true
  blendMode=BLEND_Translucent
  pixelDepthOffsetMode=PDOM_Legacy
  previewMesh=&/SM_Detail_Skull
  usedShadingModels=Default Lit
 }
 nodes [
  n1 = ParticleColor()
  n2 = Desaturation(
   None=TextureSample() {texture=/Game/Effects/Fort_Effects/GameplayPulginsAssets/Skull/SkullDetail/Texture/T_DetailSkull_E}.RGB
   Fraction=ScalarParameter() {parameterName=Des}
  )
  MaterialFunctionCall(Emissive=Multiply(A=Multiply(A=@n1.RGB, B=@n2), B=ScalarParameter() {parameterName=Emissive}) {constB=2}) {materialFunction=&MF_FX_EmissiveAdjustments}
  Multiply(
   A=Multiply(A=ParticleColor().A)
   B=MaterialFunctionCall("FadeSensitivity (optional)"=ScalarParameter() {defaultValue=3, parameterName="FadeSensitivity (optional) (S)"}) {materialFunction=&FadeBasedOnViewAngle}.Result
  )
  Multiply(A=@n2, B=@n1.RGB)
 ]
}
